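# Material function that adds animated lights to the emissive color of incoming material attributes for the SolarTheory backpack.
# Offset, scaled pre-skinned position, computed per vertex and interpolated, supplies mesh-locked coordinates for a stepped position mask and a radial angle.
# The angle repeats by NumberOfLights and is quantized into segments, and time scaled by LightSpeed selects the lit one, producing chasing lights.
# Mask and segment are multiplied, tinted by LightColor and added to EmissiveColor.
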MaterialFunction /BRCosmetics/Accessories/FORT_Backpacks/Backpack_M_MED_SolarTheory/Materials/MF_SolarTheoryBackpackEmissive
graph {nodes [
  n1 = GetMaterialAttributes(None=FunctionInput() {inputType=FunctionInput_MaterialAttributes}) {attributeGetTypes [#1]}
  n2 = NamedRerouteUsage()
  n3 = Step(Y=VectorParameter() {parameterName=StepValue, defaultValue=<rgba 1 0 0 1>}.RGB, X=Abs(None=@n2))
  n4 = Multiply(A=Add(A=ComponentMask(None=@n3) {g=true}, B=ComponentMask(None=@n3) {r=true}), B=OneMinus(None=ComponentMask(None=@n2) {b=true}))
  n5 = Frac(None=Multiply(A=Time(), B=ScalarParameter() {defaultValue=1, parameterName=LightSpeed}))
  n6 = NamedRerouteUsage()
  n7 = ComponentMask(None=MaterialFunctionCall(Vector or UVs=ComponentMask(None=@n6) {g=true, r=true}) {materialFunction=&VectorToRadialValue}.Radial Coordinates) {r=true}
  n8 = ScalarParameter() {defaultValue=2, parameterName=NumberOfLights}
  n9 = Divide(B=@n8) {constA=16}
  n10 = Multiply(A=Frac(None=Multiply(A=@n7, B=@n8)), B=@n9) {constB=16}
  n11 = Floor(None=Add(A=@n10))
  n12 = Floor(None=@n10)
  n13 = Subtract(A=Step(Y=@n5, X=Divide(A=@n11, B=@n9) {constB=16}) {constY=1}, B=Step(Y=@n5, X=Divide(A=@n12, B=@n9) {constB=16}) {constY=1})
  n14 = Multiply(A=VectorParameter() {parameterName=LightColor, defaultValue=<rgba 1 1 1 1>}.RGB, B=Saturate(None=Multiply(A=@n4, B=@n13)))
  FunctionOutput(None=SetMaterialAttributes!(MaterialAttributes=@n1.MaterialAttributes, EmissiveColor=Add(A=@n1.EmissiveColor, B=@n14)) {attributeSetTypes [#1]})
  n15 = Add(A=PreSkinnedPosition(), B=VectorParameter() {parameterName=MaskPosition, defaultValue=<rgba 0 0 0 1>}.RGB)
  n16 = VectorParameter() {parameterName=LocalPositionMultiplier, defaultValue=<rgba 10.8 1 1 1>}
  NamedRerouteDeclaration(None=VertexInterpolator(VS=Divide(A=@n15, B=@n16.RGB) {constB=1}).PS) {name=PreSkinnedUv, nodeColor=<rgba 1 0.9609 0 1>}
 ]}
