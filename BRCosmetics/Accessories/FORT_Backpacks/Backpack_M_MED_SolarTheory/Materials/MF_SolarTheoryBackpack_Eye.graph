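# Material function for the SolarTheory backpack that adds animated, color-ramped eyes as emissive, toggled by the UseEyeAnimation static switch.
# A time-driven flipbook samples the eye texture on UV1, with a step on pre-skinned position apparently offsetting some eyes by half a cycle.
# Smoothstep thresholds turn the animation into layered colors with a dark cutoff.
# Shared anime-master LUT and time-of-day functions adjust the color before it replaces EmissiveColor.
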MaterialFunction /BRCosmetics/Accessories/FORT_Backpacks/Backpack_M_MED_SolarTheory/Materials/MF_SolarTheoryBackpack_Eye
graph {nodes [
  n1 = GetMaterialAttributes(None=NamedRerouteUsage()) {attributeGetTypes [#1]}
  n2 = MaterialFunctionCall!(Color=MaterialFunctionCall!(In=NamedRerouteUsage(), LUT, LUT Multiplier) {materialFunction=/Game/Packages/Fortress_SharedMaterials/Anime_Master_Material/MF_ApplyLUT}.Result) {materialFunction=/Game/Packages/Fortress_SharedMaterials/Anime_Master_Material/MF_TODColorAdjustment}
  FunctionOutput(None=StaticSwitchParameter(
    True=SetMaterialAttributes!(MaterialAttributes=@n1.MaterialAttributes, EmissiveColor=@n2.Result) {attributeSetTypes [#2]}
    False=NamedRerouteUsage()
   ) {parameterName=UseEyeAnimation})
  NamedRerouteDeclaration(None=FunctionInput() {inputType=FunctionInput_MaterialAttributes}) {name=Material, nodeColor=<rgba 0.6172 0 1 1>}
  n3 = PreSkinnedPosition()
  NamedRerouteDeclaration(None=Step(X=VertexInterpolator(VS=@n3).PS)) {name=PreSkinnedUv, nodeColor=<rgba 1 0.9609 0 1>}
  n4 = Multiply(A=Time(), B=ScalarParameter() {defaultValue=1.7, parameterName=EyeFlipbookSpeed}) {constB=2}
  n5 = Add(A=@n4) {constB=0.5}
  NamedRerouteDeclaration(None=LinearInterpolate(A=Frac(None=@n4), B=Frac(None=@n5), Alpha=NamedRerouteUsage())) {name=EyeDesync, nodeColor=<rgba 0 1 0.4141 1>}
  n6 = TextureSampleParameter2D(UVs=MaterialFunctionCall(
    "Animation  Phase (0-1)"=NamedRerouteUsage()
    Number of Rows=ScalarParameter() {defaultValue=3, parameterName=EyeFlipbookRows}
    Number of Columns=ScalarParameter() {defaultValue=3, parameterName=EyeFlipbookColumns}
    UVs=TextureCoordinate() {coordinateIndex=1}
   ) {materialFunction=&FlipBook}.UVs) {
   parameterName=EyeFlipbookTexture
   texture=/BRCosmetics/Accessories/FORT_Backpacks/Backpack_M_MED_SolarTheory/Textures/T_SolarTheoryBackpackEye
  }
  NamedRerouteDeclaration(None=@n6.RGB) {name=EyeAnimation, nodeColor=<rgba 0.6406 1 0 1>}
  n7 = LinearInterpolate(
   A=Constant3Vector() {constant=<rgba 0.4427 0.3557 0 1>}
   B=Constant3Vector() {constant=<rgba 1 0.5068 0.026 1>}
   Alpha=SmoothStep(Value=NamedRerouteUsage()) {constMax=0.08}
  )
  n8 = SmoothStep(Value=NamedRerouteUsage()) {constMax=0.2}
  n9 = LinearInterpolate(A=Constant3Vector() {constant=<rgba 1 0.7526 0.2448 1>}, B=Constant3Vector() {constant=<rgba 0.6376 0 0.0033 1>}, Alpha=@n8)
  n10 = SmoothStep(Value=NamedRerouteUsage()) {constMin=0.6}
  NamedRerouteDeclaration(None=Multiply(A=LinearInterpolate(A=@n7, B=@n9, Alpha=@n8), B=OneMinus(None=@n10))) {name=EyeColors, nodeColor=<rgba 0 1 0.5547 1>}
 ]}
