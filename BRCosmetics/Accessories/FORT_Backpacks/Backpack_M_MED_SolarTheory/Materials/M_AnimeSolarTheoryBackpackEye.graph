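# Anime-shaded character material for the eye of the Solar Theory backpack cosmetic, using material attributes and the Cloth shading model on skeletal meshes.
# PreFX feeds anime shading, then a chain of visor, hair puff, bobbing, flipbook mouth and cloth-sim functions.
# A backpack-specific eye function follows, then anime SkinFX, which takes the shading, ink-line and FX masks, before PostFX.
# A mobile toon outline function is present but unconnected.
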Material /BRCosmetics/Accessories/FORT_Backpacks/Backpack_M_MED_SolarTheory/Materials/M_AnimeSolarTheoryBackpackEye
graph {
 properties {
  bAutomaticallySetUsageInEditor=false
  bUseMaterialAttributes=true
  bUsedWithMeshDeformer=true
  bUsedWithSkeletalMesh=true
  pixelDepthOffsetMode=PDOM_Legacy
  shadingModel=MSM_Cloth
  usedShadingModels=Cloth
 }
 nodes [
  MaterialFunctionCall!(Material) {materialFunction=/Game/Packages/Fortress_SharedMaterials/Anime_Master_Material/MF_ToonOutline_Mobile}
  n1 = MaterialFunctionCall() {materialFunction=&BaseFN_CharacterFunction_PreFX}
  n2 = MaterialFunctionCall!(
   Material=@n1.Result
   LitDiffuseOverride
   SSC_AlphaOverride
   Tint Mask
   DF InkLine Visibility Mask
   DF Inkline Injection
   DF Inkline Color Override
   DF Inkline Color Override Mask
   CustomUVs
  ) {materialFunction=/Game/Packages/Fortress_SharedMaterials/Anime_Master_Material/MF_AnimeShading}
  n3 = MaterialFunctionCall!(In=MaterialFunctionCall!(In=@n2.Result) {materialFunction=/Game/Packages/Fortress_SharedMaterials/Anime_Master_Material/MF_AnimatedVisor}.Result) {materialFunction=/Game/Characters/Player/Female/Medium/Bodies/F_MED_Peppermint/Materials/MF_HairPuffWPO}
  n4 = MaterialFunctionCall!(
   Material=MaterialFunctionCall!(In=MaterialFunctionCall!(In=@n3.Result) {materialFunction=/Game/Packages/Fortress_SharedMaterials/Anime_Master_Material/MF_BobbingWPO}.Result) {materialFunction=/Game/Packages/Fortress_SharedMaterials/Anime_Master_Material/MF_ToonFlipbookMouth}.Result
   SkinFX_Mask=@n2.FX
  ) {materialFunction=/Game/Global/MaterialFunctions/MF_ClothSimWPO_v2}
  MaterialFunctionCall(
   AO=@n1.AO
   Distance Blend=@n1.AO
   Fresnel=@n1.AO
   Gmap=@n1.AO
   Pre Skinned Local Position=@n1.AO
   Material=MaterialFunctionCall!(
    MA=MaterialFunctionCall!(In=@n4.Result) {materialFunction=&/MF_SolarTheoryBackpack_Eye}.Result
    Shading=@n2.Shading
    InkLine Mask=@n2.InkLine Mask
    SSC=@n2.SSC
    SkinFX=@n2.FX
   ) {materialFunction=/Game/Packages/Fortress_SharedMaterials/Anime_Master_Material/MF_AnimeSkinFX}.MA
  ) {materialFunction=&BaseFN_CharacterFunction_PostFX}
 ]
}
