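# Physics asset for the SolarTheory backpack cosmetic, covering a small figure with head, body, hands, feet, collar tags and belt flap on the pack.
# Every bone uses a capsule body, and body and head are constrained to the pack and to each other with angular limits.
# Hands, feet and collar base have free angular joints, while finger and toe chains are angularly locked.
# Both tags and the belt flap hang off the collar base with limited swing.
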PhysicsAsset /BRCosmetics/Accessories/FORT_Backpacks/Backpack_M_MED_SolarTheory/Meshes/M_MED_SolarTheory_Pack_Physics
bodies {
 base_collar_tags [{shapeType=Capsule, length=0.505, radius=0.505, shapeName=base_collar_tags_capsule, rotation=<rot 0 -3.4545 -81.8396>}]
 base_foot_l [{shapeType=Capsule, length=0, radius=1.5, shapeName=base_foot_l_capsule, center=<1 0 0>, rotation=<rot 90 0 0>}]
 base_foot_r [{shapeType=Capsule, length=0, radius=1.5, shapeName=base_foot_r_capsule, center=<1 0 0>, rotation=<rot 90 0 0>}]
 base_hand_l [{shapeType=Capsule, length=0, radius=1.25, shapeName=base_hand_l_capsule, center=<1 0 0>, rotation=<rot 90 0 0>}]
 base_hand_r [{shapeType=Capsule, length=0, radius=1.25, shapeName=base_hand_r_capsule, center=<1 0 0>, rotation=<rot 90 0 0>}]
 dyn_belt_flap [{shapeType=Capsule, length=1, radius=0.25, shapeName=dyn_belt_flap_capsule, center=<1 0 0>, rotation=<rot 90 0 0>}]
 dyn_body [{shapeType=Capsule, length=0.5, radius=4, shapeName=dyn_body_capsule, center=<-1 0 0>, rotation=<rot 90 0 0>}]
 dyn_circular_tag [{shapeType=Capsule, length=0.7, radius=0.07, shapeName=dyn_circular_tag_capsule, center=<0.3 0 0>, rotation=<rot 90 0 0>}]
 dyn_head [{shapeType=Capsule, length=0, radius=3, shapeName=dyn_head_capsule, center=<2 0 0>, rotation=<rot 90 0 0>}]
 dyn_index_01_l [{shapeType=Capsule, length=1, radius=0.25, shapeName=dyn_index_01_l_capsule, center=<1 0 0>, rotation=<rot 90 0 0>}]
 dyn_index_01_r [{shapeType=Capsule, length=1, radius=0.25, shapeName=dyn_index_01_r_capsule, center=<1 0 0>, rotation=<rot 90 0 0>}]
 dyn_index_toe_01_l [{shapeType=Capsule, length=1.5, radius=0.25, shapeName=dyn_index_toe_01_l_capsule, center=<1 0 0>, rotation=<rot 90 0 0>}]
 dyn_index_toe_01_r [{shapeType=Capsule, length=1.5, radius=0.25, shapeName=dyn_index_toe_01_r_capsule, center=<1.5 0 0>, rotation=<rot 90 0 0>}]
 dyn_middle_01_l [{shapeType=Capsule, length=1, radius=0.25, shapeName=dyn_middle_01_l_capsule, center=<1.5 0 0>, rotation=<rot 90 0 0>}]
 dyn_middle_01_r [{shapeType=Capsule, length=1, radius=0.25, shapeName=dyn_middle_01_r_capsule, center=<1 0 0>, rotation=<rot 90 0 0>}]
 dyn_middle_02_l [{shapeType=Capsule, length=1, radius=0.25, shapeName=dyn_middle_02_l_capsule, center=<1 0 0>, rotation=<rot 90 0 0>}]
 dyn_middle_02_r [{shapeType=Capsule, length=1, radius=0.25, shapeName=dyn_middle_02_r_capsule, center=<1 0 0>, rotation=<rot 90 0 0>}]
 dyn_middle_toe_01_l [{shapeType=Capsule, length=1.5, radius=0.25, shapeName=dyn_middle_toe_01_l_capsule, center=<1 0 0>, rotation=<rot 90 0 0>}]
 dyn_middle_toe_01_r [{shapeType=Capsule, length=1.5, radius=0.25, shapeName=dyn_middle_toe_01_r_capsule, center=<1.5 0 0>, rotation=<rot 90 0 0>}]
 dyn_middle_toe_02_l [{shapeType=Capsule, length=1.5, radius=0.25, shapeName=dyn_middle_toe_02_l_capsule, center=<1 0 0>, rotation=<rot 90 0 0>}]
 dyn_middle_toe_02_r [{shapeType=Capsule, length=1.5, radius=0.25, shapeName=dyn_middle_toe_02_r_capsule, center=<1 0 0>, rotation=<rot 90 0 0>}]
 dyn_rectangular_tag [{shapeType=Capsule, length=1.3, radius=0.07, shapeName=dyn_rectangular_tag_capsule, center=<0.5 0 0>, rotation=<rot 90 0 0>}]
 dyn_ring_01_l [{shapeType=Capsule, length=1, radius=0.25, shapeName=dyn_ring_01_l_capsule, center=<1 0 0>, rotation=<rot 90 0 0>}]
 dyn_ring_01_r [{shapeType=Capsule, length=1, radius=0.25, shapeName=dyn_ring_01_r_capsule, center=<1 0 0>, rotation=<rot 90 0 0>}]
 dyn_ring_toe_01_l [{shapeType=Capsule, length=1.5, radius=0.25, shapeName=dyn_ring_toe_01_l_capsule, center=<1 0 0>, rotation=<rot 90 0 0>}]
 dyn_ring_toe_01_r [{shapeType=Capsule, length=1.5, radius=0.25, shapeName=dyn_ring_toe_01_r_capsule, center=<1.5 0 0>, rotation=<rot 90 0 0>}]
 dyn_ring_toe_02_l [{shapeType=Capsule, length=1.5, radius=0.25, shapeName=dyn_ring_toe_02_l_capsule, center=<1 0 0>, rotation=<rot 90 0 0>}]
 dyn_ring_toe_02_r [{shapeType=Capsule, length=1.5, radius=0.25, shapeName=dyn_ring_toe_02_r_capsule, center=<1 0 0>, rotation=<rot 90 0 0>}]
 pack [{shapeType=Capsule, length=0, radius=9, shapeName=pack_capsule, center=<0 -11 -5>, rotation=<rot 90 0 0>}]
}
constraints [
 {bone1Name=dyn_circular_tag, bone2Name=base_collar_tags, swing1LimitDegrees=25, swing2LimitDegrees=45, twistLimitDegrees=1}
 {bone1Name=dyn_rectangular_tag, bone2Name=base_collar_tags, swing1LimitDegrees=15, swing2LimitDegrees=45, twistLimitDegrees=1}
 {bone1Name=dyn_belt_flap, bone2Name=base_collar_tags, swing1LimitDegrees=20, swing2LimitDegrees=10, twistLimitDegrees=1}
 {
  bone1Name=base_foot_l
  bone2Name=pack
  swing1LimitDegrees=45
  swing1Motion=Free
  swing2LimitDegrees=45
  swing2Motion=Free
  twistLimitDegrees=45
  twistMotion=Free
 }
 {
  bone1Name=dyn_index_toe_01_l
  bone2Name=base_foot_l
  swing1LimitDegrees=5
  swing1Motion=Locked
  swing2LimitDegrees=5
  swing2Motion=Locked
  twistLimitDegrees=1
  twistMotion=Locked
 }
 {
  bone1Name=dyn_middle_toe_01_l
  bone2Name=base_foot_l
  swing1LimitDegrees=5
  swing1Motion=Locked
  swing2LimitDegrees=5
  swing2Motion=Locked
  twistLimitDegrees=1
  twistMotion=Locked
 }
 {
  bone1Name=dyn_middle_toe_02_l
  bone2Name=dyn_middle_toe_01_l
  swing1LimitDegrees=5
  swing1Motion=Locked
  swing2LimitDegrees=5
  swing2Motion=Locked
  twistLimitDegrees=1
  twistMotion=Locked
 }
 {
  bone1Name=dyn_ring_toe_01_l
  bone2Name=base_foot_l
  swing1LimitDegrees=5
  swing1Motion=Locked
  swing2LimitDegrees=10
  swing2Motion=Locked
  twistLimitDegrees=1
  twistMotion=Locked
 }
 {
  bone1Name=dyn_ring_toe_02_l
  bone2Name=dyn_ring_toe_01_l
  swing1LimitDegrees=5
  swing1Motion=Locked
  swing2LimitDegrees=10
  swing2Motion=Locked
  twistLimitDegrees=1
  twistMotion=Locked
 }
 {
  bone1Name=base_foot_r
  bone2Name=pack
  swing1LimitDegrees=45
  swing1Motion=Free
  swing2LimitDegrees=45
  swing2Motion=Free
  twistLimitDegrees=45
  twistMotion=Free
 }
 {
  bone1Name=dyn_index_toe_01_r
  bone2Name=base_foot_r
  swing1LimitDegrees=5
  swing1Motion=Locked
  swing2LimitDegrees=5
  swing2Motion=Locked
  twistLimitDegrees=1
  twistMotion=Locked
 }
 {
  bone1Name=dyn_middle_toe_01_r
  bone2Name=base_foot_r
  swing1LimitDegrees=5
  swing1Motion=Locked
  swing2LimitDegrees=5
  swing2Motion=Locked
  twistLimitDegrees=1
  twistMotion=Locked
 }
 {
  bone1Name=dyn_middle_toe_02_r
  bone2Name=dyn_middle_toe_01_r
  swing1LimitDegrees=5
  swing1Motion=Locked
  swing2LimitDegrees=5
  swing2Motion=Locked
  twistLimitDegrees=1
  twistMotion=Locked
 }
 {
  bone1Name=dyn_ring_toe_01_r
  bone2Name=base_foot_r
  swing1LimitDegrees=5
  swing1Motion=Locked
  swing2LimitDegrees=10
  swing2Motion=Locked
  twistLimitDegrees=1
  twistMotion=Locked
 }
 {
  bone1Name=dyn_ring_toe_02_r
  bone2Name=dyn_ring_toe_01_r
  swing1LimitDegrees=5
  swing1Motion=Locked
  swing2LimitDegrees=10
  swing2Motion=Locked
  twistLimitDegrees=1
  twistMotion=Locked
 }
 {
  bone1Name=base_hand_l
  bone2Name=pack
  swing1LimitDegrees=45
  swing1Motion=Free
  swing2LimitDegrees=45
  swing2Motion=Free
  twistLimitDegrees=45
  twistMotion=Free
 }
 {
  bone1Name=dyn_index_01_l
  bone2Name=base_hand_l
  swing1LimitDegrees=10
  swing1Motion=Locked
  swing2LimitDegrees=10
  swing2Motion=Locked
  twistLimitDegrees=1
  twistMotion=Locked
 }
 {
  bone1Name=dyn_middle_01_l
  bone2Name=base_hand_l
  swing1LimitDegrees=10
  swing1Motion=Locked
  swing2LimitDegrees=10
  swing2Motion=Locked
  twistLimitDegrees=1
  twistMotion=Locked
 }
 {
  bone1Name=dyn_middle_02_l
  bone2Name=dyn_middle_01_l
  swing1LimitDegrees=10
  swing1Motion=Locked
  swing2LimitDegrees=10
  swing2Motion=Locked
  twistLimitDegrees=1
  twistMotion=Locked
 }
 {
  bone1Name=dyn_ring_01_l
  bone2Name=base_hand_l
  swing1LimitDegrees=10
  swing1Motion=Locked
  swing2LimitDegrees=10
  swing2Motion=Locked
  twistLimitDegrees=1
  twistMotion=Locked
 }
 {
  bone1Name=base_hand_r
  bone2Name=pack
  swing1LimitDegrees=45
  swing1Motion=Free
  swing2LimitDegrees=45
  swing2Motion=Free
  twistLimitDegrees=45
  twistMotion=Free
 }
 {
  bone1Name=dyn_index_01_r
  bone2Name=base_hand_r
  swing1LimitDegrees=10
  swing1Motion=Locked
  swing2LimitDegrees=10
  swing2Motion=Locked
  twistLimitDegrees=1
  twistMotion=Locked
 }
 {
  bone1Name=dyn_middle_01_r
  bone2Name=base_hand_r
  swing1LimitDegrees=10
  swing1Motion=Locked
  swing2LimitDegrees=10
  swing2Motion=Locked
  twistLimitDegrees=1
  twistMotion=Locked
 }
 {
  bone1Name=dyn_middle_02_r
  bone2Name=dyn_middle_01_r
  swing1LimitDegrees=10
  swing1Motion=Locked
  swing2LimitDegrees=10
  swing2Motion=Locked
  twistLimitDegrees=1
  twistMotion=Locked
 }
 {
  bone1Name=dyn_ring_01_r
  bone2Name=base_hand_r
  swing1LimitDegrees=10
  swing1Motion=Locked
  swing2LimitDegrees=10
  swing2Motion=Locked
  twistLimitDegrees=1
  twistMotion=Locked
 }
 {bone1Name=dyn_body, bone2Name=pack, swing1LimitDegrees=45, swing2LimitDegrees=120, twistLimitDegrees=90}
 {bone1Name=dyn_head, bone2Name=pack, swing1LimitDegrees=47.5, swing2LimitDegrees=21, twistLimitDegrees=1}
 {bone1Name=dyn_head, bone2Name=dyn_body, swing1LimitDegrees=5, swing2LimitDegrees=60, twistLimitDegrees=1}
 {
  bone1Name=base_collar_tags
  bone2Name=dyn_body
  swing1LimitDegrees=45
  swing1Motion=Free
  swing2LimitDegrees=45
  swing2Motion=Free
  twistLimitDegrees=45
  twistMotion=Free
 }
 {
  bone1Name=base_collar_tags
  bone2Name=dyn_head
  swing1LimitDegrees=45
  swing1Motion=Free
  swing2LimitDegrees=45
  swing2Motion=Free
  twistLimitDegrees=45
  twistMotion=Free
 }
]
